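# Additive, unlit, two-sided particle material for the candle flame on the Birthday 2019 pickaxe.
# A panned flow map, offset per particle by FlameOffset and weighted by a UV sine gradient, distorts the flame mask.
# Panning swirl noise modulates the mask, which is squared and indexes a color ramp, tinted by particle color.
# By default, the ramp lookup scales down as the camera looks vertically, suiting flat flame cards.
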Material /BRCosmetics/Weapons/FORT_Melee/Pickaxe_Birthday_2019/FX/M_Birthday2019PickaxeCandleFlame
graph {
 properties {
  bAutomaticallySetUsageInEditor=false
  bUsedWithNiagaraMeshParticles=true
  bUsedWithNiagaraSprites=true
  bUsedWithParticleSprites=true
  bUsedWithStaticMesh=false
  blendMode=BLEND_Additive
  pixelDepthOffsetMode=PDOM_Legacy
  shadingModel=MSM_Unlit
  twoSided=true
  usedShadingModels=Unlit
 }
 nodes [
  n1 = ParticleColor()
  n2 = Sine(None=TextureCoordinate() {uTiling=0.5, vTiling=0.5})
  n3 = Multiply(A=ComponentMask(None=@n2) {r=true}, B=ComponentMask(None=@n2) {g=true})
  n4 = Panner(
   Coordinate=Add(A=TextureCoordinate() {uTiling=0.2, vTiling=0.2}, B=DynamicParameter() {paramNames [FlameOffset,FlameSpeed,Param3,Param4]}.FlameOffset)
   Speed=VectorParameter() {parameterName=FlowPanSpeed, defaultValue=<rgba -0.5 0.1 0 0>}.RGB
  ) {speedX=-0.5, speedY=0.1}
  n5 = ComponentMask(None=Subtract(
    A=TextureSampleParameter2D(UVs=@n4) {parameterName=FlowMap, texture=&T_NoisePack_01}.RGB
    B=ScalarParameter() {defaultValue=0.4, parameterName=FlowErodeAmount}
   ) {constB=0.4}) {g=true, r=true}
  n6 = Add(A=TextureCoordinate(), B=Multiply(A=Multiply(A=@n3, B=@n5)) {constB=0.4})
  n7 = TextureSampleParameter2D(UVs=@n6) {parameterName=MaskAltTexture, texture=/Game/Effects/Fort_Effects/Textures/Misc/t_feather_down}
  LinearInterpolate(A=@n1.A, B=Multiply(A=@n1.A, B=@n7.A) {constB=1.5}, Alpha=ScalarParameter() {parameterName=UseMaskAlpha})
  n8 = Panner(
   Coordinate=MaterialFunctionCall(
    UVs=TextureCoordinate() {uTiling=0.8, vTiling=0.4}
    Texture Scale=VectorParameter() {parameterName=DetailUVScale, defaultValue=<rgba 1 1 0 0>}.RGB
   ) {materialFunction=&ScaleUVsByCenter}.UVs
   Speed=VectorParameter() {parameterName=DetailPanSpeed, defaultValue=<rgba -1.5 -0.2 0 0>}.RGB
  ) {speedX=-1.5, speedY=-0.2}
  n9 = TextureCoordinate() {uTiling=0.2, vTiling=0.2}
  n10 = VectorParameter() {parameterName=FlowDetailPanSpeed, defaultValue=<rgba -0.5 0.1 0 0>}
  n11 = ComponentMask(None=TextureSampleParameter2D(UVs=Panner(Coordinate=@n9, Speed=@n10.RGB) {speedX=-0.5, speedY=0.1}) {parameterName=DetailFlowMap, samplerType=SAMPLERTYPE_Grayscale, texture=&LowResBlurredNoise}.RGB) {g=true, r=true}
  n12 = TextureSampleParameter2D(UVs=LinearInterpolate(A=@n8, B=Multiply(A=@n8, B=@n11), Alpha=ScalarParameter() {parameterName=UseDetailFlowMap})) {parameterName=DetailTexture, samplerType=SAMPLERTYPE_LinearGrayscale, texture=/Game/Global/Textures/Default/Noise/T_SwirlNoise_N_height}
  n13 = Multiply(
   A=LinearInterpolate(
    A=TextureSampleParameter2D(UVs=@n6) {parameterName=MaskTexture, samplerType=SAMPLERTYPE_LinearColor, texture=/Game/Effects/Fort_Effects/Textures/Fire/Candle}.R
    B=@n7.R
    Alpha=ScalarParameter() {parameterName=UseMaskAlpha}
   )
   B=LinearInterpolate(Alpha=@n12.R) {constA=0.5, constB=1.5}
  )
  n14 = Multiply(A=@n13, B=@n13)
  n15 = LinearInterpolate(A=@n14, B=Add(A=@n14, B=@n7.A), Alpha=ScalarParameter() {parameterName=UseMaskAlpha})
  n16 = Constant()
  n17 = ComponentMask(None=Transform(None=Constant3Vector() {constant=<rgba 0 0 1 0>}) {transformSourceType=TRANSFORMSOURCE_Camera}) {b=true}
  n18 = Multiply(A=OneMinus(None=Abs(None=@n17)), B=@n15)
  n19 = TextureSampleParameter2D(UVs=LinearInterpolate(A=@n15, B=AppendVector(A=@n16, B=@n18), Alpha=ScalarParameter() {defaultValue=1, parameterName=UseWorldSpace})) {parameterName=ColorRamp, samplerSource=SSM_Clamp_WorldGroupSettings, texture=/Game/Effects/Fort_Effects/Textures/Fire/FlameRamp}
  MaterialFunctionCall(Emissive=Multiply(A=@n19.RGB, B=@n1.RGB) {constB=1.5}) {materialFunction=&MF_FX_EmissiveAdjustments}
 ]
}
