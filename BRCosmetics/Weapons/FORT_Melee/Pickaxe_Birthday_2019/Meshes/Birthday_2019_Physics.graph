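# Physics asset for the Birthday 2019 pickaxe mesh, giving box bodies to its two-handed handle, fork prong and two ribbons for secondary motion.
# The fork prong is constrained to the handle with tight angular limits, allowing only slight wobble.
# Each ribbon is a three-link chain with wide swing and twist limits, branching from a shared ribbon master body, which has no constraint to the handle.
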PhysicsAsset /BRCosmetics/Weapons/FORT_Melee/Pickaxe_Birthday_2019/Meshes/Birthday_2019_Physics
bodies {
 Dyn_fork_prong [{shapeType=Box, extentX=60.1813, extentY=55.8341, extentZ=33.4705, center=<34.2485 9.9276 -1.0709>}]
 Dyn_ribbon_a_01 [{shapeType=Box, extentX=2, extentY=7.1845, extentZ=9.0764, center=<4 0 0>, rotation=<rot 90 0 0>}]
 Dyn_ribbon_a_02 [{shapeType=Box, extentX=2, extentY=6.2886, extentZ=7.6337, center=<4 0 0>, rotation=<rot 90 0 0>}]
 Dyn_ribbon_a_03 [{shapeType=Box, extentX=2, extentY=5.1219, extentZ=7.0474, center=<4 0 0>, rotation=<rot 90 0 0>}]
 Dyn_ribbon_b_01 [{shapeType=Box, extentX=2, extentY=7.4678, extentZ=8.3619, center=<4 0 0>, rotation=<rot 90 0 0>}]
 Dyn_ribbon_b_02 [{shapeType=Box, extentX=2, extentY=6.0749, extentZ=8.1614, center=<4 0 0>, rotation=<rot 90 0 0>}]
 Dyn_ribbon_b_03 [{shapeType=Box, extentX=2, extentY=5.0918, extentZ=6, center=<4 0 0>, rotation=<rot 90 0 0>}]
 Dyn_ribbon_mstr [{shapeType=Box, extentX=7.5, extentY=7.5, extentZ=7.5}]
 Melee_TwoHanded_Handle [{shapeType=Box, extentX=2.183, extentY=0.5, extentZ=96.5619, center=<0.8832 -1.2388 4.8228>}]
}
constraints [
 {bone1Name=Dyn_fork_prong, bone2Name=Melee_TwoHanded_Handle, swing1LimitDegrees=5, swing2LimitDegrees=5, twistLimitDegrees=5}
 {bone1Name=Dyn_ribbon_a_01, bone2Name=Dyn_ribbon_mstr, swing1LimitDegrees=65, swing2LimitDegrees=65, twistLimitDegrees=65}
 {bone1Name=Dyn_ribbon_a_02, bone2Name=Dyn_ribbon_a_01, swing1LimitDegrees=65, swing2LimitDegrees=65, twistLimitDegrees=65}
 {bone1Name=Dyn_ribbon_a_03, bone2Name=Dyn_ribbon_a_02, swing1LimitDegrees=65, swing2LimitDegrees=65, twistLimitDegrees=65}
 {bone1Name=Dyn_ribbon_b_01, bone2Name=Dyn_ribbon_mstr, swing1LimitDegrees=65, swing2LimitDegrees=65, twistLimitDegrees=65}
 {bone1Name=Dyn_ribbon_b_02, bone2Name=Dyn_ribbon_b_01, swing1LimitDegrees=65, swing2LimitDegrees=65, twistLimitDegrees=65}
 {bone1Name=Dyn_ribbon_b_03, bone2Name=Dyn_ribbon_b_02, swing1LimitDegrees=65, swing2LimitDegrees=65, twistLimitDegrees=65}
]
