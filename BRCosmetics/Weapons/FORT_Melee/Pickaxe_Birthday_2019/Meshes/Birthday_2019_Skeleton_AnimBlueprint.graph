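# Animation Blueprint for the 2019 Birthday pickaxe skeleton that adds physics-driven secondary motion to the cosmetic melee weapon.
# The input pose runs through a RigidBody node using the Birthday_2019 physics asset, overridden gravity and scaled component velocity and acceleration, then to the output.
# Two chained Trail controllers for the ribbon bones, with planar limits, are left unconnected to the output.
# The event graph holds no logic.
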AnimBlueprint /BRCosmetics/Weapons/FORT_Melee/Pickaxe_Birthday_2019/Meshes/Birthday_2019_Skeleton_AnimBlueprint
target_skeleton=&/Birthday_2019_Skeleton
graphs [
 AnimGraph {nodes [
   AnimGraphNode_SubInput_0 = Input Pose()
   n1 = RigidBody(ComponentPose=Local To Component(LocalPose=@AnimGraphNode_SubInput_0)) {
    bOverrideWorldGravity=true
    overridePhysicsAsset=&/Birthday_2019_Physics
    alphaBoolBlend {blendOption=Linear}
    componentAppliedLinearAccClamp=<2000 2000 2000>
    componentLinearAccScale=<0.4 0.4 0.4>
    componentLinearVelScale=<4 0.4 0.4>
    overrideWorldGravity {z=-980}
   }
   Output Pose(Result=Component To Local(ComponentPose=@n1)) {layerGroup=None}
   n2 = Trail controller - Bone: Dyn_ribbon_a_03!(ComponentPose, Alpha=1, FakeVelocity:Vector="0,0,-20") -> (Pose) {
    bActorSpaceFakeVel=true
    bUsePlanarLimit=true
    chainLength=3
    lODThreshold=4
    maxDeltaTime=0.0333
    relaxationSpeedScale=1.5
    baseJoint {boneName=Dyn_ribbon_mstr}
    planarLimits [
     {drivingBone {boneName=Dyn_ribbon_mstr}, planeTransform {location=<0 -2.0567 0>, rotation=<rot 0 0 -90>, scale=<1 1 1>}}
     {drivingBone {boneName=Dyn_ribbon_mstr}, planeTransform {location=<0 0 -10>, rotation=<rot 60 -180 30.0001>, scale=<1 1 1>}}
    ]
    rotationLimits [{limitMax=<180 180 180>}, {limitMax=<180 180 180>}, {limitMax=<180 180 180>}]
    rotationOffsets [<0 0 0>, <0 0 0>, <0 0 0>]
    trailBone {boneName=Dyn_ribbon_a_03}
   }
   Trail controller - Bone: Dyn_ribbon_b_03!(ComponentPose=@n2, Alpha=1, FakeVelocity:Vector="0,0,-50") -> (Pose) {
    bActorSpaceFakeVel=true
    bUsePlanarLimit=true
    chainLength=3
    lODThreshold=4
    maxDeltaTime=0.0333
    relaxationSpeedScale=1.5
    baseJoint {boneName=Dyn_ribbon_mstr}
    planarLimits [
     {drivingBone {boneName=Dyn_ribbon_mstr}, planeTransform {location=<0 -2.0567 0>, rotation=<rot 0 0 -90>, scale=<1 1 1>}}
     {drivingBone {boneName=Dyn_ribbon_mstr}, planeTransform {location=<0 0 -10>, rotation=<rot 90 114.0361 0>, scale=<1 1 1>}}
    ]
    rotationLimits [{limitMax=<180 180 180>}, {limitMax=<180 180 180>}, {limitMax=<180 180 180>}]
    rotationOffsets [<0 0 0>, <0 0 0>, <0 0 0>]
    trailBone {boneName=Dyn_ribbon_b_03}
   }
  ]}
 EventGraph {nodes [
   TryGetPawnOwner()
   Event BlueprintUpdateAnimation()
  ]}
]
variables []
